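# EQS query for the NPC LeashTeleport ability that finds a location, presumably the teleport destination.
# A small SimpleGrid generates candidate points, projected vertically onto navigation using the AnyTraversable nav filter.
# Distance tests discard points too close to nearby AI pawns or any player.
# Remaining points are scored by inverse linear 3D distance to the querier, favoring the closest spot.
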EnvQuery /NPCLibrary/Abilities/LeashTeleport/EQS/EQS_LeashTeleport_FindLocation
options [{
  generator {
   name=EnvQueryGenerator_SimpleGrid_0
   class=EnvQueryGenerator_SimpleGrid
   properties {
    gridSize=400
    spaceBetween=300
    projectionData {
     extentX=0
     navigationFilter=/Game/AI/NavFilters/NavFilter_AnyTraversable.NavFilter_AnyTraversable_C
     postProjectionVerticalOffset=0
     projectDown=1024
     projectUp=1024
    }
   }
  }
  tests [
   EnvQueryTest_Distance_3 {
    class=EnvQueryTest_Distance
    properties {
     distanceTo=&EnvQueryContext_Querier
     filterType=Range
     floatValueMin=0
     scoringEquation=InverseLinear
     scoringFactor=1
     testComment=""
     testMode=Distance3D
     testPurpose=Score
    }
   }
   EnvQueryTest_Distance_4 {
    class=EnvQueryTest_Distance
    properties {
     distanceTo=&FortQueryContext_NearbyAIPawns
     filterType=Minimum
     floatValueMin=200
     scoringEquation=Linear
     scoringFactor=1
     testComment=""
     testMode=Distance3D
     testPurpose=Filter
    }
   }
   EnvQueryTest_Distance_5 {
    class=EnvQueryTest_Distance
    properties {
     distanceTo=/Script/FortniteAIServer.FortQueryContext_AllPlayers
     filterType=Minimum
     floatValueMin=200
     scoringEquation=Linear
     scoringFactor=1
     testComment=""
     testMode=Distance3D
     testPurpose=Filter
    }
   }
  ]
 }]
